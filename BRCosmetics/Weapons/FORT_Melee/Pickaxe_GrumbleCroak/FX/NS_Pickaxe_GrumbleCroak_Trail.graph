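# Niagara ribbon trail for the GrumbleCroak pickaxe, following the weapon's animation through anim-trail bone or socket sampling.
# Two emitters layer a wide, faint, longer-lived ribbon with a narrower, sharper, shorter-lived one.
# Both taper over lifetime while color curves shift from hot pink through yellow-green to teal as alpha fades out.
# A system-level time-of-day multiplier scales both ribbons' alpha.
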NiagaraSystem /BRCosmetics/Weapons/FORT_Melee/Pickaxe_GrumbleCroak/FX/NS_Pickaxe_GrumbleCroak_Trail
dependencies {
 usedDataInterfaces [&NiagaraDataInterfaceCurve, &NiagaraDataInterfaceColorCurve]
 usedDynamicInputs [&FNTimeOfDayScalar]
 usedModules [
  &SystemState
  &self:SystemSpawnScript.SystemScriptSource.SystemScriptGraph.NiagaraNodeAssignment_0.SetVariables_#1
 ]
}
emitters {
 Trail_Faint {
  emitterUpdateScript [
   &EmitterState()
   &FN_SpawnAnimTrail_v2(AnimTrailWidthMode=::FromSecond)
  ]
  particleSpawnScript [
   /Game/Effects/Niagara/Modules/AnimTrail/FN_InitializeAnimTrail/FN_InitializeAnimTrailBoneIndecies(Second Bone or Socket Index=1)
   /Game/Effects/Niagara/Modules/AnimTrail/FN_InitializeAnimTrail/FN_InitializeAnimTrail(AnimTrailWidth=1, Color=<rgba 3 0.181 1.6317 1>, Lifetime=0.3)
   &SystemLocation(MaskSpawnGroup=false)
  ]
  particleUpdateScript [
   &ScaleRibbonWidth(Ribbon Width Scale=&FloatFromCurve(FloatCurve=DataInterface() {propertyValues {Curve {keys [
         {arriveTangent=-1.2579, interpMode=RCIM_Cubic, time=0, value=1}
         {arriveTangent=-0.9664, interpMode=RCIM_Cubic, leaveTangent=-0.9664, tangentMode=RCTM_User, time=1, value=0}
        ]}}}))
   &FN_UpdateAnimTrail_v2()
   &Color(
    Color=&ColorFromCurve(ColorCurve=DataInterface() {propertyValues {
       AlphaCurve {keys [<at 0.169 0.15 RCIM_Cubic>,<at 1 0 RCIM_Cubic>]}
       BlueCurve {keys [<at 0.0743 1.6317 RCIM_Cubic>,<at 0.2735 0>,<at 0.4662 0.7318>]}
       GreenCurve {keys [<at 0.0743 0.181 RCIM_Cubic>,<at 0.2735 1>,<at 0.4662 0.896>]}
       RedCurve {keys [<at 0.0743 3 RCIM_Cubic>,<at 0.2735 0.8189>,<at 0.4662 0>]}
      }})
    Scale Alpha=$System.TimeOfDayColorMultiplier
    Scale Color=<2.5 2.5 2.5>
   )
  ]
  rendererClasses [&NiagaraRibbonRendererProperties]
  renderers [{rendererClass=&NiagaraRibbonRendererProperties}]
 }
 Trail_Sharp {
  emitterUpdateScript [
   &EmitterState()
   &FN_SpawnAnimTrail_v2(AnimTrailWidthMode=::FromSecond)
  ]
  particleSpawnScript [
   &FN_InitializeAnimTrailBoneIndecies(Second Bone or Socket Index=1)
   &FN_InitializeAnimTrail(AnimTrailWidth=0.6, Color=<rgba 3 0.181 1.6317 1>, Lifetime=0.18)
   &SystemLocation(MaskSpawnGroup=false)
  ]
  particleUpdateScript [
   &ScaleRibbonWidth(Ribbon Width Scale=&FloatFromCurve(FloatCurve=DataInterface() {propertyValues {Curve {keys [
         {arriveTangent=-0.979, interpMode=RCIM_Cubic, leaveTangent=-0.979, tangentMode=RCTM_User, time=0, value=1}
         {arriveTangent=-0.1336, interpMode=RCIM_Cubic, leaveTangent=-0.1336, tangentMode=RCTM_User, time=1, value=0}
        ]}}}))
   &FN_UpdateAnimTrail_v2()
   &Color(
    Color=&ColorFromCurve(ColorCurve=DataInterface() {propertyValues {
       AlphaCurve {keys [<at 0.4499 0.33>,<at 1 0 RCIM_Cubic>]}
       BlueCurve {keys [<at 0.2784 3.2634>,<at 0.4597 0>,<at 0.6458 0.5107>]}
       GreenCurve {keys [<at 0.2784 0.3621>,<at 0.4597 1.5>,<at 0.6458 0.707>]}
       RedCurve {keys [<at 0.2784 6>,<at 0.4597 1.2283>,<at 0.6458 0>]}
      }})
    Scale Alpha=$System.TimeOfDayColorMultiplier
   )
  ]
  rendererClasses [&NiagaraRibbonRendererProperties]
  renderers [{rendererClass=&NiagaraRibbonRendererProperties}]
 }
}
summary {
 systemName=NS_Pickaxe_GrumbleCroak_Trail
 emitters [
  {emitterName=Trail_Faint, rendererClasses [&NiagaraRibbonRendererProperties]}
  {emitterName=Trail_Sharp, rendererClasses [&NiagaraRibbonRendererProperties]}
 ]
 userVariables [
  User.RibbonIDOffset {type=NiagaraInt32}
  User.SkeletalMesh {
   type=NiagaraDataInterfaceSkeletalMesh
   defaultValue {
    struct=NiagaraExt_VariableValue_DataInterface
    value {dataInterface=&self:NiagaraDataInterfaceSkeletalMesh_7}
   }
  }
 ]
}
